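# Animation Blueprint for the SleekRivet male medium cosmetic body, adding secondary motion to the pose copied from its attached parent mesh.
# A Control Rig driven by DBNO, skydive, glide and slide states feeds chained AnimDynamics for jacket, amulet and zippers.
# Constraints then adjust shoulders and amulet base, with cape-only RigidBody physics and an attach-bone offset.
# The EventGraph updates those states and drives material WPO cloth, adding wind while skydiving or parachuting.
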AnimBlueprint /BRCosmetics/Characters/Player/Male/Medium/Bodies/M_MED_SleekRivet/Meshes/M_MED_SleekRivet_AnimBP
target_skeleton=&/M_MED_SleekRivet_Skeleton
graphs [
 AnimGraph {nodes [
   Anim Dynamics - Bone: dyn_jacket_back_01!(ComponentPose, bAlphaBoolEnabled:bool=True, GravityOverride:Vector="0,0,-980", ExternalForce="0,0,0") -> (Pose) {
    alphaInputType=Bool
    angularDampingOverride=1
    angularSpringConstant=500
    bAngularSpring=true
    bGravityOverrideInSimSpace=true
    bOverrideAngularDamping=true
    bUseGravityOverride=true
    bUseSphericalLimits=true
    simulationSpace=RootRelative
    alphaBoolBlend {blendOption=Linear}
    alphaScaleBiasClamp {bClampResult=true}
    boundBone {boneName=dyn_jacket_back_01}
    componentAppliedLinearAccClamp=<5000 5000 5000>
    componentLinearAccScale=<0.3 0.3 0.3>
    componentLinearVelScale=<0.3 0.3 0.3>
    gravityOverride {z=-980}
    physicsBodyDefinitions [{
      collisionType=CustomSphere
      sphereCollisionRadius=3
      boundBone {boneName=dyn_jacket_back_01}
      boxExtents=<6 6 6>
      constraintSetup {coneAngle=1, angularLimitsMax=<3 -11 10>, angularLimitsMin=<-3 -3 -10>, angularTarget=<1 0 0>}
      localJointOffset=<0.1 0 0>
     }]
    relativeSpaceBone {boneName=spine_04}
    sphericalLimits [{limitRadius=9, drivingBone {boneName=spine_03}, sphereLocalOffset=<0 -0.707 -3.7759>}]
   }
   "Constraint - Dyn_amulet_02 (P)"!(ComponentPose, Alpha=1, ConstraintWeights_0=1, ConstraintWeights_1:float=0) -> (Pose) {
    alphaBoolBlend {blendOption=Linear}
    boneToModify {boneName=Dyn_amulet_02}
    constraintSetup [{targetBone {boneName=Dyn_amulet_01}}]
    constraintWeights [1,0]
   }
   "Constraint - dyn_Shoulder_bk_l (PPPPPPPPPPP)"!(
    ComponentPose
    Alpha=1
    ConstraintWeights_0=0.0384
    ConstraintWeights_1:float=0.2097
    ConstraintWeights_2:float=0.2404
    ConstraintWeights_3:float=0.0062
    ConstraintWeights_4:float=0.0738
    ConstraintWeights_5:float=0.0023
    ConstraintWeights_6:float=0.1139
    ConstraintWeights_7:float=0.149
    ConstraintWeights_8:float=0.0043
    ConstraintWeights_9:float=0.018
    ConstraintWeights_10:float=0.1442
   ) -> (Pose) {
    alphaBoolBlend {blendOption=Linear}
    boneToModify {boneName=dyn_Shoulder_bk_l}
    constraintSetup [
     {targetBone {boneName=spine_03}}
     {targetBone {boneName=spine_05}}
     {targetBone {boneName=clavicle_l}}
     {targetBone {boneName=upperarm_l}}
     {targetBone {boneName=upperarm_twist_01_l}}
     {targetBone {boneName=upperarm_twist_02_l}}
     {targetBone {boneName=deform_upperarm_bk_l}}
     {targetBone {boneName=deform_upperarm_up_l}}
     {targetBone {boneName=Armor_correction_mid_l}}
     {targetBone {boneName=Armor_correction_up_l}}
     {targetBone {boneName=dyn_Shoulder_bk_l}}
    ]
    constraintWeights [1,1,1,1,1,1,1,1,1,1,1]
   }
   n1 = Control Rig!(
    Source=Copy Pose From Mesh() {bUseAttachedParent=true}
    Alpha
    Is DBNO:bool=Get bIsDBNO()
    Sky alpha:double=Get Sky Aplha() -> (Sky Aplha:double)
    HangTwoArms:double=Get HangTwo arms() -> (HangTwo arms:double)
    Slide alpha:double=Get Slide Alpha() -> (Slide Alpha:double)
   ) -> (Pose) {
    alphaBoolBlend {blendOption=Linear}
    controlRigAssetReference {blueprintRigClass=&/M_MED_SleekRivet_CtrlRig.M_MED_SleekRivet_CtrlRig_C}
   }
   n2 = Anim Dynamics - Bone: dyn_jacket_front_l_01!(ComponentPose=Local To Component(LocalPose=@n1), Alpha=0.7, GravityOverride:Vector="0,0,-980", ExternalForce="0,0,0") -> (Pose) {
    angularDampingOverride=1
    angularSpringConstant=500
    bAngularSpring=true
    bGravityOverrideInSimSpace=true
    bOverrideAngularDamping=true
    bUseGravityOverride=true
    bUseSphericalLimits=true
    simulationSpace=RootRelative
    alphaBoolBlend {blendOption=Linear}
    alphaScaleBiasClamp {bClampResult=true}
    boundBone {boneName=dyn_jacket_front_l_01}
    componentAppliedLinearAccClamp=<5000 5000 5000>
    componentLinearAccScale=<0.3 0.3 0.3>
    componentLinearVelScale=<0.3 0.3 0.3>
    gravityOverride {z=-980}
    physicsBodyDefinitions [{
      collisionType=CustomSphere
      sphereCollisionRadius=3
      boundBone {boneName=dyn_jacket_front_l_01}
      boxExtents=<9 9 9>
      constraintSetup {coneAngle=1, angularLimitsMax=<7 20 4>, angularLimitsMin=<-5 -3 -20>, angularTarget=<1 0 0>}
      localJointOffset=<0.6 0 0>
     }]
    relativeSpaceBone {boneName=spine_04}
    sphericalLimits [{limitRadius=9, drivingBone {boneName=spine_03}, sphereLocalOffset=<0 0 0>}]
   }
   n3 = Anim Dynamics - Bone: dyn_jacket_side_l_01!(ComponentPose=@n2, Alpha=0.7, GravityOverride:Vector="0,0,-980", ExternalForce="0,0,0") -> (Pose) {
    angularDampingOverride=1
    angularSpringConstant=500
    bAngularSpring=true
    bGravityOverrideInSimSpace=true
    bOverrideAngularDamping=true
    bUseGravityOverride=true
    bUseSphericalLimits=true
    simulationSpace=RootRelative
    alphaBoolBlend {blendOption=Linear}
    alphaScaleBiasClamp {bClampResult=true}
    boundBone {boneName=dyn_jacket_side_l_01}
    componentAppliedLinearAccClamp=<5000 5000 5000>
    componentLinearAccScale=<0.3 0.3 0.3>
    componentLinearVelScale=<0.3 0.3 0.3>
    gravityOverride {z=-980}
    physicsBodyDefinitions [{
      collisionType=CustomSphere
      sphereCollisionRadius=3
      boundBone {boneName=dyn_jacket_side_l_01}
      boxExtents=<8 8 8>
      constraintSetup {coneAngle=1, angularLimitsMax=<0 22 7>, angularLimitsMin=<0 -3 -7>, angularTarget=<1 0 0>}
      localJointOffset=<0.6 0 0>
     }]
    relativeSpaceBone {boneName=spine_04}
    sphericalLimits [{limitRadius=9, drivingBone {boneName=spine_03}, sphereLocalOffset=<0.1716 -0.707 -2.4453>}]
   }
   n4 = Anim Dynamics - Bone: dyn_jacket_back_l_01!(ComponentPose=@n3, bAlphaBoolEnabled:bool=True, GravityOverride:Vector="0,0,-980", ExternalForce="0,0,0") -> (Pose) {
    alphaInputType=Bool
    angularDampingOverride=1
    angularSpringConstant=500
    bAngularSpring=true
    bGravityOverrideInSimSpace=true
    bOverrideAngularDamping=true
    bUseGravityOverride=true
    bUseSphericalLimits=true
    simulationSpace=RootRelative
    alphaBoolBlend {blendOption=Linear}
    alphaScaleBiasClamp {bClampResult=true}
    boundBone {boneName=dyn_jacket_back_l_01}
    componentAppliedLinearAccClamp=<5000 5000 5000>
    componentLinearAccScale=<0.3 0.3 0.3>
    componentLinearVelScale=<0.3 0.3 0.3>
    gravityOverride {z=-980}
    physicsBodyDefinitions [{
      collisionType=CustomSphere
      sphereCollisionRadius=3
      boundBone {boneName=dyn_jacket_back_l_01}
      boxExtents=<6 6 6>
      constraintSetup {coneAngle=1, angularLimitsMax=<5 11 7>, angularLimitsMin=<-5 -5 -7>, angularTarget=<1 0 0>}
      localJointOffset=<0.6 0 0>
     }]
    relativeSpaceBone {boneName=spine_04}
    sphericalLimits [{limitRadius=9, drivingBone {boneName=spine_03}, sphereLocalOffset=<0 -0.707 -3.7759>}]
   }
   n5 = Anim Dynamics - Bone: dyn_jacket_front_r_01!(ComponentPose=@n4, Alpha=0.7, GravityOverride:Vector="0,0,-980", ExternalForce="0,0,0") -> (Pose) {
    angularDampingOverride=1
    angularSpringConstant=500
    bAngularSpring=true
    bGravityOverrideInSimSpace=true
    bOverrideAngularDamping=true
    bUseGravityOverride=true
    bUseSphericalLimits=true
    simulationSpace=RootRelative
    alphaBoolBlend {blendOption=Linear}
    alphaScaleBiasClamp {bClampResult=true}
    boundBone {boneName=dyn_jacket_front_r_01}
    componentAppliedLinearAccClamp=<5000 5000 5000>
    componentLinearAccScale=<0.3 0.3 0.3>
    componentLinearVelScale=<0.3 0.3 0.3>
    gravityOverride {z=-980}
    physicsBodyDefinitions [{
      collisionType=CustomSphere
      sphereCollisionRadius=3
      boundBone {boneName=dyn_jacket_front_r_01}
      boxExtents=<9 9 9>
      constraintSetup {coneAngle=1, angularLimitsMax=<7 20 20>, angularLimitsMin=<-5 -3 -4>, angularTarget=<1 0 0>}
      localJointOffset=<0.6 0 0>
     }]
    relativeSpaceBone {boneName=spine_04}
    sphericalLimits [{limitRadius=8, drivingBone {boneName=thigh_r}, sphereLocalOffset=<19.3237 -5.5566 -0.0465>}]
   }
   n6 = Anim Dynamics - Bone: dyn_jacket_side_r_01!(ComponentPose=@n5, Alpha=0.7, GravityOverride:Vector="0,0,-980", ExternalForce="0,0,0") -> (Pose) {
    angularDampingOverride=1
    angularSpringConstant=500
    bAngularSpring=true
    bGravityOverrideInSimSpace=true
    bOverrideAngularDamping=true
    bUseGravityOverride=true
    bUseSphericalLimits=true
    simulationSpace=RootRelative
    alphaBoolBlend {blendOption=Linear}
    alphaScaleBiasClamp {bClampResult=true}
    boundBone {boneName=dyn_jacket_side_r_01}
    componentAppliedLinearAccClamp=<5000 5000 5000>
    componentLinearAccScale=<0.3 0.3 0.3>
    componentLinearVelScale=<0.3 0.3 0.3>
    gravityOverride {z=-980}
    physicsBodyDefinitions [{
      collisionType=CustomSphere
      sphereCollisionRadius=3
      boundBone {boneName=dyn_jacket_side_r_01}
      boxExtents=<8 8 8>
      constraintSetup {coneAngle=1, angularLimitsMax=<13 22 4>, angularLimitsMin=<-5 -3 -4>, angularTarget=<1 0 0>}
      localJointOffset=<0.6 0 0>
     }]
    relativeSpaceBone {boneName=spine_04}
    sphericalLimits [{limitRadius=9, drivingBone {boneName=spine_03}, sphereLocalOffset=<1.1518 -1.0501 3.6164>}]
   }
   n7 = Anim Dynamics - Bone: dyn_jacket_back_r_01!(ComponentPose=@n6, bAlphaBoolEnabled:bool=True, GravityOverride:Vector="0,0,-980", ExternalForce="0,0,0") -> (Pose) {
    alphaInputType=Bool
    angularDampingOverride=1
    angularSpringConstant=500
    bAngularSpring=true
    bGravityOverrideInSimSpace=true
    bOverrideAngularDamping=true
    bUseGravityOverride=true
    bUseSphericalLimits=true
    simulationSpace=RootRelative
    alphaBoolBlend {blendOption=Linear}
    alphaScaleBiasClamp {bClampResult=true}
    boundBone {boneName=dyn_jacket_back_r_01}
    componentAppliedLinearAccClamp=<5000 5000 5000>
    componentLinearAccScale=<0.3 0.3 0.3>
    componentLinearVelScale=<0.3 0.3 0.3>
    gravityOverride {z=-980}
    physicsBodyDefinitions [{
      collisionType=CustomSphere
      sphereCollisionRadius=3
      boundBone {boneName=dyn_jacket_back_r_01}
      boxExtents=<6 6 6>
      constraintSetup {coneAngle=1, angularLimitsMax=<5 11 7>, angularLimitsMin=<-5 -5 -7>, angularTarget=<1 0 0>}
      localJointOffset=<0.6 0 0>
     }]
    relativeSpaceBone {boneName=spine_04}
    sphericalLimits [{limitRadius=9, drivingBone {boneName=spine_03}, sphereLocalOffset=<-0.0987 -2.3729 1.8018>}]
   }
   n8 = Anim Dynamics - Bone: Dyn_amulet_01!(ComponentPose=@n7, bAlphaBoolEnabled:bool=True, ExternalForce="0,0,0") -> (Pose) {
    alphaInputType=Bool
    angularDampingOverride=0.9999
    angularSpringConstant=500
    bAngularSpring=true
    bGravityOverrideInSimSpace=true
    bOverrideAngularDamping=true
    bUseGravityOverride=true
    bUseSphericalLimits=true
    simulationSpace=BoneRelative
    alphaBoolBlend {blendOption=Linear}
    boundBone {boneName=Dyn_amulet_01}
    componentAppliedLinearAccClamp=<3500 3500 3500>
    componentLinearAccScale=<1.2 1.2 1.2>
    componentLinearVelScale=<1.2 1.2 1.2>
    gravityOverride {z=-980}
    physicsBodyDefinitions [{
      collisionType=CustomSphere
      sphereCollisionRadius=2
      boundBone {boneName=Dyn_amulet_01}
      boxExtents=<15 20 20>
      constraintSetup {angularLimitsMax=<1 1 6>, angularLimitsMin=<-1 -60 -6>, angularTarget=<0 0 0>}
      localJointOffset=<4 0 0>
     }]
    relativeSpaceBone {boneName=root}
    sphericalLimits [{limitRadius=10, drivingBone {boneName=spine_05}, sphereLocalOffset=<0 0 0>}]
   }
   n9 = Anim Dynamics - Bone: Dyn_amulet_02!(ComponentPose=@n8, bAlphaBoolEnabled:bool=True, ExternalForce="0,0,0") -> (Pose) {
    alphaInputType=Bool
    angularDampingOverride=0.9999
    angularSpringConstant=500
    bAngularSpring=true
    bGravityOverrideInSimSpace=true
    bOverrideAngularDamping=true
    bUseGravityOverride=true
    bUseSphericalLimits=true
    simulationSpace=BoneRelative
    alphaBoolBlend {blendOption=Linear}
    boundBone {boneName=Dyn_amulet_02}
    componentAppliedLinearAccClamp=<3500 3500 3500>
    componentLinearAccScale=<1.2 1.2 1.2>
    componentLinearVelScale=<1.2 1.2 1.2>
    gravityOverride {z=-980}
    physicsBodyDefinitions [{
      collisionType=CustomSphere
      sphereCollisionRadius=2
      boundBone {boneName=Dyn_amulet_02}
      boxExtents=<15 20 20>
      constraintSetup {angularLimitsMax=<1 -1 7>, angularLimitsMin=<-1 -70 -7>, angularTarget=<0 0 0>}
      localJointOffset=<4 0 0>
     }]
    relativeSpaceBone {boneName=root}
   }
   n10 = Anim Dynamics - Bone: dyn_zipper_jacket!(ComponentPose=@n9, Alpha=1, GravityOverride:Vector="0,0,-980", ExternalForce="0,0,0") -> (Pose) {
    angularDampingOverride=1
    angularSpringConstant=500
    bAngularSpring=true
    bOverrideAngularDamping=true
    bUseGravityOverride=true
    gravityScale=0.35
    simulationSpace=RootRelative
    alphaBoolBlend {blendOption=Linear}
    boundBone {boneName=dyn_zipper_jacket}
    componentAppliedLinearAccClamp=<6000 6000 6000>
    componentLinearAccScale=<1 1 1>
    componentLinearVelScale=<1 1 1>
    gravityOverride {z=-980}
    physicsBodyDefinitions [{
      boundBone {boneName=dyn_zipper_jacket}
      boxExtents=<10 10 10>
      constraintSetup {angularLimitsMax=<15 15 80>, angularLimitsMin=<-15 -15 -15>, angularTarget=<1 0 0>}
      localJointOffset=<1.2 0 0>
     }]
   }
   n11 = Anim Dynamics - Bone: dyn_zipper_glute_r(ComponentPose=@n10, Alpha=1, ExternalForce="0,0,0") {
    angularDampingOverride=0.99
    angularSpringConstant=500
    bOverrideAngularDamping=true
    bUseGravityOverride=true
    bUsePlanarLimit=false
    gravityScale=-1
    simulationSpace=RootRelative
    alphaBoolBlend {blendOption=Linear}
    boundBone {boneName=dyn_zipper_glute_r}
    componentAppliedLinearAccClamp=<5000 5000 5000>
    componentLinearAccScale=<0.2 0.2 0.2>
    componentLinearVelScale=<0.2 0.2 0.2>
    gravityOverride {z=-980}
    physicsBodyDefinitions [{
      collisionType=CustomSphere
      sphereCollisionRadius=1
      boundBone {boneName=dyn_zipper_glute_r}
      boxExtents=<10 5 5>
      constraintSetup {angularLimitsMax=<10 0 40>, angularLimitsMin=<0 -60 -5>, angularTarget=<0 0 0>}
      localJointOffset=<1 -0.1 0>
     }]
   }
   n12 = Anim Dynamics - Bone: dyn_zipper_glute_l(ComponentPose=@n11, Alpha=1, ExternalForce="0,0,0") {
    angularDampingOverride=0.99
    angularSpringConstant=500
    bOverrideAngularDamping=true
    bUseGravityOverride=true
    bUsePlanarLimit=false
    gravityScale=-1
    simulationSpace=RootRelative
    alphaBoolBlend {blendOption=Linear}
    boundBone {boneName=dyn_zipper_glute_l}
    componentAppliedLinearAccClamp=<5000 5000 5000>
    componentLinearAccScale=<0.2 0.2 0.2>
    componentLinearVelScale=<0.2 0.2 0.2>
    gravityOverride {z=-980}
    physicsBodyDefinitions [{
      collisionType=CustomSphere
      sphereCollisionRadius=1
      boundBone {boneName=dyn_zipper_glute_l}
      boxExtents=<10 5 5>
      constraintSetup {angularLimitsMax=<10 0 40>, angularLimitsMin=<0 -60 -5>, angularTarget=<0 0 0>}
      localJointOffset=<1 -0.1 0>
     }]
   }
   n13 = Anim Dynamics - Bone: dyn_zipper_calf_l(ComponentPose=@n12, Alpha=1, ExternalForce="0,0,0") {
    angularDampingOverride=0.99
    angularSpringConstant=500
    bOverrideAngularDamping=true
    bUseGravityOverride=true
    bUsePlanarLimit=false
    gravityScale=-1
    simulationSpace=RootRelative
    alphaBoolBlend {blendOption=Linear}
    boundBone {boneName=dyn_zipper_calf_l}
    componentAppliedLinearAccClamp=<5000 5000 5000>
    componentLinearAccScale=<0.2 0.2 0.2>
    componentLinearVelScale=<0.2 0.2 0.2>
    gravityOverride {z=-980}
    physicsBodyDefinitions [{
      collisionType=CustomSphere
      sphereCollisionRadius=1
      boundBone {boneName=dyn_zipper_calf_l}
      boxExtents=<6 5 5>
      constraintSetup {coneAngle=1, twistAxis=AxisZ, angularLimitsMax=<10 0 50>, angularLimitsMin=<0 -60 1>, angularTarget=<0 0 1>}
      localJointOffset=<1 0 0>
     }]
   }
   n14 = Anim Dynamics - Bone: dyn_zipper_calf_r(ComponentPose=@n13, Alpha=1, ExternalForce="0,0,0") {
    angularDampingOverride=0.99
    angularSpringConstant=500
    bOverrideAngularDamping=true
    bUseGravityOverride=true
    bUsePlanarLimit=false
    gravityScale=-1
    simulationSpace=RootRelative
    alphaBoolBlend {blendOption=Linear}
    boundBone {boneName=dyn_zipper_calf_r}
    componentAppliedLinearAccClamp=<5000 5000 5000>
    componentLinearAccScale=<0.2 0.2 0.2>
    componentLinearVelScale=<0.2 0.2 0.2>
    gravityOverride {z=-980}
    physicsBodyDefinitions [{
      collisionType=CustomSphere
      sphereCollisionRadius=1
      boundBone {boneName=dyn_zipper_calf_r}
      boxExtents=<10 5 5>
      constraintSetup {angularLimitsMax=<10 30 50>, angularLimitsMin=<0 -10 -1>, angularTarget=<0 0 0>}
      localJointOffset=<1 0 0>
     }]
   }
   n15 = Anim Dynamics - Bone: dyn_zipper_l(ComponentPose=@n14, Alpha=1, ExternalForce="0,0,0") {
    angularDampingOverride=0.99
    angularSpringConstant=500
    bOverrideAngularDamping=true
    bUseGravityOverride=true
    bUsePlanarLimit=false
    bUseSphericalLimits=true
    gravityScale=-1
    simulationSpace=RootRelative
    alphaBoolBlend {blendOption=Linear}
    boundBone {boneName=dyn_zipper_l}
    componentAppliedLinearAccClamp=<5000 5000 5000>
    componentLinearAccScale=<0.2 0.2 0.2>
    componentLinearVelScale=<0.2 0.2 0.2>
    gravityOverride {z=-980}
    physicsBodyDefinitions [{
      collisionType=CustomSphere
      sphereCollisionRadius=1
      boundBone {boneName=dyn_zipper_l}
      boxExtents=<10 5 5>
      constraintSetup {angularLimitsMax=<10 22 40>, angularLimitsMin=<0 -60 -1>, angularTarget=<0 0 0>}
      localJointOffset=<1 -0.3 0>
     }]
    sphericalLimits [{limitRadius=6.8, drivingBone {boneName=thigh_l}, sphereLocalOffset=<-8.6489 1.1547 6.6057>}]
   }
   n16 = Anim Dynamics - Bone: dyn_zipper_r(ComponentPose=@n15, Alpha=1, ExternalForce="0,0,0") {
    angularDampingOverride=0.99
    angularSpringConstant=500
    bOverrideAngularDamping=true
    bUseGravityOverride=true
    bUsePlanarLimit=false
    gravityScale=-1
    simulationSpace=RootRelative
    alphaBoolBlend {blendOption=Linear}
    boundBone {boneName=dyn_zipper_r}
    componentAppliedLinearAccClamp=<5000 5000 5000>
    componentLinearAccScale=<0.2 0.2 0.2>
    componentLinearVelScale=<0.2 0.2 0.2>
    gravityOverride {z=-980}
    physicsBodyDefinitions [{
      collisionType=CustomSphere
      sphereCollisionRadius=1
      boundBone {boneName=dyn_zipper_r}
      boxExtents=<10 5 5>
      constraintSetup {angularLimitsMax=<10 0 40>, angularLimitsMin=<0 -60 -1>, angularTarget=<0 0 0>}
      localJointOffset=<1 0 0>
     }]
   }
   n17 = "Constraint - dyn_Shoulder_bk_r (PPPPPPPPPPPPP)"!(
    ComponentPose=@n16
    Alpha=1
    ConstraintWeights_0=0.0235
    ConstraintWeights_1:float=0.0109
    ConstraintWeights_2:float=0.1186
    ConstraintWeights_3:float=0.0005
    ConstraintWeights_4:float=0.2435
    ConstraintWeights_5:float=0.005
    ConstraintWeights_6:float=0.0371
    ConstraintWeights_7:float=0.0034
    ConstraintWeights_8:float=0.0846
    ConstraintWeights_9:float=0.0709
    ConstraintWeights_10:float=0.0067
    ConstraintWeights_11:float=0.0141
    ConstraintWeights_12:float=0.3812
   ) -> (Pose) {
    alphaBoolBlend {blendOption=Linear}
    boneToModify {boneName=dyn_Shoulder_bk_r}
    constraintSetup [
     {targetBone {boneName=spine_03}}
     {targetBone {boneName=spine_04}}
     {targetBone {boneName=spine_05}}
     {targetBone {boneName=neck_02}}
     {targetBone {boneName=clavicle_r}}
     {targetBone {boneName=upperarm_r}}
     {targetBone {boneName=upperarm_twist_01_r}}
     {targetBone {boneName=upperarm_twist_02_r}}
     {targetBone {boneName=deform_upperarm_bk_r}}
     {targetBone {boneName=deform_upperarm_up_r}}
     {targetBone {boneName=Armor_correction_mid_r}}
     {targetBone {boneName=Armor_correction_up_r}}
     {targetBone {boneName=dyn_Shoulder_bk_r}}
    ]
    constraintWeights [1,1,1,1,1,1,1,1,1,1,1,1,1]
   }
   n18 = "Constraint - dyn_Shoulder_bk_l (PPPPPPPPPPPPP)"!(
    ComponentPose=@n17
    Alpha=1
    ConstraintWeights_0=0.0002
    ConstraintWeights_1:float=0.1448
    ConstraintWeights_2:float=0.1125
    ConstraintWeights_3:float=0.1842
    ConstraintWeights_4:float=0.2331
    ConstraintWeights_5:float=0.0233
    ConstraintWeights_6:float=0.064
    ConstraintWeights_7:float=0.0978
    ConstraintWeights_8:float=0.0112
    ConstraintWeights_9:float=0.0172
    ConstraintWeights_10:float=0.0486
    ConstraintWeights_11:float=0.0199
    ConstraintWeights_12:float=0.0433
   ) -> (Pose) {
    alphaBoolBlend {blendOption=Linear}
    boneToModify {boneName=dyn_Shoulder_bk_l}
    constraintSetup [
     {targetBone {boneName=spine_02}}
     {targetBone {boneName=spine_03}}
     {targetBone {boneName=spine_04}}
     {targetBone {boneName=spine_05}}
     {targetBone {boneName=clavicle_l}}
     {targetBone {boneName=upperarm_l}}
     {targetBone {boneName=upperarm_twist_01_l}}
     {targetBone {boneName=deform_upperarm_bk_l}}
     {targetBone {boneName=deform_upperarm_up_l}}
     {targetBone {boneName=Armor_correction_dwn_l}}
     {targetBone {boneName=Armor_correction_mid_l}}
     {targetBone {boneName=Armor_correction_up_l}}
     {targetBone {boneName=dyn_Shoulder_bk_l}}
    ]
    constraintWeights [1,1,1,1,1,1,1,1,1,1,1,1,1]
   }
   n19 = RigidBody!(ComponentPose=@n18, bAlphaBoolEnabled:bool=Get is Wearing Cape() -> (is Wearing Cape:bool), ExternalForce) -> (Pose) {
    alphaInputType=Bool
    overridePhysicsAsset=&/M_MED_SleekRivet_Physics_Cape
    alphaBoolBlend {blendOption=Linear}
   }
   n20 = "Constraint - Base_amulet_01 (PP)"!(ComponentPose=@n19, Alpha=1, ConstraintWeights_0=0.6, ConstraintWeights_1:float=0.4) -> (Pose) {
    alphaBoolBlend {blendOption=Linear}
    boneToModify {boneName=Base_amulet_01}
    constraintSetup [{targetBone {boneName=spine_05}}, {targetBone {boneName=neck_01}}]
    constraintWeights [1,1]
   }
   n21 = "Transform (Modify) Bone - Bone: attach_cape"(ComponentPose=@n20, Alpha=1, Translation="0,-4,0", Rotation="0,0,0", Scale="1,1,1") {
    translationMode=BMM_Additive
    translationSpace=BCS_BoneSpace
    alphaBoolBlend {blendOption=Linear}
    boneToModify {boneName=attach_cape}
    translation {y=-4}
   }
   Output Pose(Result=Component To Local(ComponentPose=@n21)) {layerGroup=None}
  ]}
 EventGraph {nodes [
   n1 = Event BlueprintUpdateAnimation()
   > n2 = Sequence!(execute) -> (then_0, then_1, then_2:Exec, then_3:Exec, then_4:Exec)
   > n3 = Gravity Override State Function DEPRECATED(
    CharacterPartAnimInstance=Self-Reference()
    GameTimeInSeconds=GetGameTimeInSeconds()
    GravityOverrideParams=Get Gravity Override Params()
    DebugJointName=dyn_hackBottomArmor
   )
   > Set Override World Gravity(Override World Gravity:Vector=@n3) -> (Output_Get:Vector)
   > Set Velocity Z(Velocity Z:double=Get PawnVelocityZ()) -> (Output_Get:double)
   n4 = Get Mesh!() -> (Mesh)
   n5 = Get Overlay Mat Instance Dyanmic() -> (Overlay Mat Instance Dyanmic:Material Instance Dynamic Object Reference)
   n6 = Get Material Instance Dynamic() -> (Material Instance Dynamic:Material Instance Dynamic Object Reference)
   > Set WPO Globals Function(
    Material=@n6
    Overlay Material=@n5
    Skeletal Mesh=@n4
    "Local Noise Transform  Location"="Get Local Noise Transform  Location"() -> ("Local Noise Transform  Location":Vector)
    Local Noise Offset=Get Local Noise Offset()
   )
   > Set WPO Regional Function!(
    execute
    Material:Material Instance Dynamic Object Reference=@n6
    Overlay Material:Material Instance Dynamic Object Reference=@n5
    Skeletal Mesh:Skeletal Mesh Component Object Reference=@n4
    Velocity Transform:Name=calf_l
    Velocity Interp Speed:double=Get Velocity Interp Speed()
    Collision Transform:Name=calf_l
    Collision Transform Axis:EAxis Enum
    Invert Collision Transform Axis:bool=false
    Delta Time X:double=@n1.DeltaTimeX
    Previous Velocity Transform Location:"Vector (by ref)"=Get R2 Velocity Transform Location()
    Previous Velocity Transform Velocity:"Vector (by ref)"=Get R2 Velocity Transform Velocity()
    Region:int=1
   ) -> (then)
   > Set WPO Regional Function(
    Material=@n6
    Overlay Material=@n5
    Skeletal Mesh=@n4
    Velocity Transform=calf_r
    Velocity Interp Speed=Get Velocity Interp Speed()
    Collision Transform=calf_r
    Invert Collision Transform Axis=true
    Delta Time X=@n1.DeltaTimeX
    Previous Velocity Transform Location=Get L Previous Velocity Transform Location() -> (L Previous Velocity Transform Location:Vector)
    Previous Velocity Transform Velocity=Get L Previous Velocity Transform Velocity() -> (L Previous Velocity Transform Velocity:Vector)
    Region=2
   )
   > SetScalarParameterValue(self=[@n6, @n5], ParameterName=ClothSim_WindStrength, Value=Get WPO Wind() -> (WPO Wind:double))
   > SetScalarParameterValue(self=[@n6, @n5], ParameterName=ClothSim_TurbulenceNoiseTileDivisor, Value=Get Turbulence Noise() -> (Turbulence Noise:float))
   > n7 = Branch(Condition=OR Boolean!(A=Get bIsSkydiveFloating(), B=Get bIsSkydiveDiveMode(), C:bool=Get bIsParachuteOpened()) -> (ReturnValue))
   > Set Velocity Interp Speed(Velocity Interp Speed:double=30) -> (Output_Get:double)
   > Set WPO Wind(WPO Wind:double=1) -> (Output_Get:double)
   > Set Turbulence Noise(Turbulence Noise:float=80) -> (Output_Get:float)
   Set Velocity Interp Speed(execute=@n7.else, Velocity Interp Speed=0)
   > Set WPO Wind(WPO Wind=0)
   > Set Turbulence Noise(Turbulence Noise=0)
   n8 = Get bIsParachuteOpened()
   n9 = Branch(execute=@n2.then_3, Condition=AND Boolean(A=Get bIsSkyDiving(), B=NOT Boolean(A=@n8)))
   > Set Sky Aplha(Sky Aplha:double=1) -> (Output_Get:double)
   Set Sky Aplha(execute=[@n16.then, @n9.else], Sky Aplha=0)
   Set is Wearing Cape(execute=@n2.then_1, is Wearing Cape:bool=CheckForBackbling().bIsBackblingACape) -> (Output_Get:bool)
   n10 = GetOwnerSkeletalMeshComponent()
   n11 = Is Valid(exec=@n2.then_2, InputObject=@n10)
   Set Is Celeb(execute=@n11.Is Not Valid, Is Celeb:bool=false) -> (Output_Get:bool)
   Set Is Celeb(
    execute=@n11.Is Valid
    Is Celeb=Montage_IsPlaying(self=GetAnimInstance(self=@n10), Montage=/Game/Animation/Game/MainPlayer/Emotes/Emote_Celebration_M)
   )
   Event OnCharacterCustomizationLoaded()
   n12 = GetPartSkeletalMeshComponent()
   > n13 = Set Mesh(Mesh:Skeletal Mesh Component Object Reference=@n12) -> (Output_Get:Skeletal Mesh Component Object Reference)
   > n14 = Get Dynamic Material Instance(SkeletalMeshComponent=@n13)
   > Set Material Instance Dynamic(Material Instance Dynamic:Material Instance Dynamic Object Reference=@n14) -> (Output_Get:Material Instance Dynamic Object Reference)
   > n15 = GetOrCreateDynamicOverlayMaterial(MeshComponent=@n12, MaterialName=None)
   > Set Overlay Mat Instance Dyanmic(Overlay Mat Instance Dyanmic:Material Instance Dynamic Object Reference=@n15) -> (Output_Get:Material Instance Dynamic Object Reference)
   n16 = Event OnEndSkydiving()
   Event OnParachuteClosed()
   > Set HangTwo arms(HangTwo arms:double=0) -> (Output_Get:double)
   Event OnEndSliding() {is=title K2Node_Event}
   > Set Slide Alpha(Slide Alpha:double=0) -> (Output_Get:double)
   Event OnBeginSliding() {is=title K2Node_Event}
   > Set Slide Alpha(Slide Alpha=1)
   Event OnParachuteOpened()
   > Switch on EGliderType(Selection=Get PlayerGliderType() -> (PlayerGliderType=HangGlider))
   > Set HangTwo arms(HangTwo arms=1)
  ]}
]
variables [
 Gravity Override Params {type=GravityOverrideParamsStruct, default {dBNO {jointName=root, gravityOverride=<0 0 -980>}}}
 HangGliding {type=boolean}
 HangTwo arms {type=number}
 Is Celeb {type=boolean}
 is Wearing Cape {type=boolean}
 L Previous Velocity Transform Location {type=Vector}
 L Previous Velocity Transform Velocity {type=Vector}
 Local Noise Offset {type=Vector}
 "Local Noise Transform  Location" {type=Vector}
 Material Instance Dynamic {type=MaterialInstanceDynamic}
 Mesh {type=SkeletalMeshComponent}
 Overlay Mat Instance Dyanmic {type=MaterialInstanceDynamic}
 Override World Gravity {type=Vector}
 R2 Velocity Transform Location {type=Vector}
 R2 Velocity Transform Velocity {type=Vector}
 Sky Aplha {type=number}
 Slide Alpha {type=number}
 Turbulence Noise {type=number}
 Velocity Interp Speed {type=number}
 Velocity Z {type=number}
 WPO Wind {type=number}
]
